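# Sound cue for the JollyTroll pickaxe's impact on an enemy.
# After a short fixed delay, a mixer blends a random pick from six main impact waves with a slightly quieter layer.
# The layer picks, with replacement, among five extra waves and a heavily weighted empty slot, so it is silent about half the time.
# It uses shared pickaxe impact attenuation, property sheet and the impacts sound class.
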SoundCue /BRCosmetics/Sounds/PickAxes/JollyTroll/PickaxeImpactEnemy_JollyTroll_Cue
root=Delay_0
nodes [
 {id=Delay_0, children [Mixer_0], properties {delayMax=0.1, delayMin=0.1}}
 {id=Mixer_0, children [Random_0,Random_1], properties {inputVolume [0.9,0.75]}}
 {id=Random_0, children [WavePlayer_0,WavePlayer_1,WavePlayer_2,WavePlayer_3,WavePlayer_4,WavePlayer_5], properties {weights [1,1,1,1,1,1]}}
 {
  id=Random_1
  children [WavePlayer_6, WavePlayer_7, WavePlayer_8, WavePlayer_9, WavePlayer_10, null]
  properties {bRandomizeWithoutReplacement=false, weights [0.2,0.2,0.2,0.2,0.2,1]}
 }
 {id=WavePlayer_0, properties {soundWaveAssetPtr=&/Pickaxe_JollyTroll_Impact_01}}
 {id=WavePlayer_1, properties {soundWaveAssetPtr=&/Pickaxe_JollyTroll_Impact_02}}
 {id=WavePlayer_10, properties {soundWaveAssetPtr=&/Pickaxe_JollyTroll_Impact_Lyr_04}}
 {id=WavePlayer_2, properties {soundWaveAssetPtr=&/Pickaxe_JollyTroll_Impact_03}}
 {id=WavePlayer_3, properties {soundWaveAssetPtr=&/Pickaxe_JollyTroll_Impact_04}}
 {id=WavePlayer_4, properties {soundWaveAssetPtr=&/Pickaxe_JollyTroll_Impact_05}}
 {id=WavePlayer_5, properties {soundWaveAssetPtr=&/Pickaxe_JollyTroll_Impact_06}}
 {id=WavePlayer_6, properties {soundWaveAssetPtr=&/Pickaxe_JollyTroll_Impact_Lyr_01}}
 {id=WavePlayer_7, properties {soundWaveAssetPtr=&/Pickaxe_JollyTroll_Impact_Lyr_02}}
 {id=WavePlayer_8, properties {soundWaveAssetPtr=&/Pickaxe_JollyTroll_Impact_Lyr_03}}
 {id=WavePlayer_9, properties {soundWaveAssetPtr=&/Pickaxe_JollyTroll_Impact_Lyr_05}}
]
properties {
 attenuationSettings=/Game/Sounds/Attenuation/Fort_Pickaxe_Impact_Att
 audioPropertiesSheet=/BRCosmetics/Sounds/Settings/PropertySheets/Pickaxe/APS_Cosmetic_Pickaxe_Impact
 maxDistance=4300
 soundClassObject=&Fort_Impacts
 concurrencySet [null]
}
